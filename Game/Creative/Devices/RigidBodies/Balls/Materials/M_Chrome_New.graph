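# Masked, Default Lit material for Creative rigid-body balls, named as a chrome finish, with dithered opacity.
# Base Color blends a Color parameter with the CreativeBall_02 texture at a fixed alpha, while the texture's green channel blends two Metallic parameters.
# Specular and Roughness are scalar parameters, and the combined attributes pass through MF_CreativeMoveTool.
# An inverted-texture tint multiply is left unconnected.
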Material /Game/Creative/Devices/RigidBodies/Balls/Materials/M_Chrome_New
graph {
 properties {bUseMaterialAttributes=true, blendMode=BLEND_Masked, ditherOpacityMask=true, pixelDepthOffsetMode=PDOM_Legacy, usedShadingModels=Default Lit}
 nodes [
  n1 = VectorParameter() {parameterName=Color, defaultValue=<rgba 0.4375 0.4375 0.4375 1>}
  n2 = TextureSample() {texture=&/T_CreativeBall_02_D}
  n3 = SetMaterialAttributes!(
   MaterialAttributes
   Base Color=LinearInterpolate(A=@n1.RGB, B=@n2.RGB)
   Metallic=LinearInterpolate(A=ScalarParameter() {parameterName=Metallic}, B=ScalarParameter() {parameterName=Metallic2}, Alpha=@n2.G)
   Specular=ScalarParameter() {parameterName=Specular}
   Roughness=ScalarParameter() {parameterName=Rough}
   Emissive Color
   Opacity Mask
   World Position Offset
  ) {attributeSetTypes [
    #1
    #2
    #3
    #4
    #5
    #6
    #7
   ]}
  MaterialFunctionCall(MA In=@n3) {materialFunction=&MF_CreativeMoveTool}
  Multiply(A=@n1.RGB, B=Clamp(None=OneMinus(None=@n2.RGB)))
 ]
}
